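# Mono, low-spec looping engine MetaSound source preset for the LionsMane vehicle, built around the shared low-spec REV parent patch.
# The parent gets the LionsMane REV asset, settings and blowoff waves, and drives the gear, RPM, throttle and shift outputs.
# A putter loop, started, stopped and gain-scaled by a crossfade segment manager on normalized speed, rises in pitch with speed and mixes with the parent's audio.
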MetaSoundSource /VehicleAudio/EngineAudio/LionsMane/MSS_EngineAudio_LionsMane_LowSpec_REV_Preset
inputs [
 Blowoff.Sound {type=WaveAsset:Array, access=Constructor}
 Blowoff.Volume {type=Float, access=Constructor}
 Character.Affiliation {type=Enum:FortAudioAffiliation}
 Character.Health {type=Float}
 Character.UniqueID {type=Int32}
 Decibels {type=Float}
 Downshift {type=Trigger}
 EngineBlowoff {type=Trigger}
 Exhaust.Pops.Blowoff.Volume {type=Float, access=Constructor}
 Exhaust.Pops.Long {type=WaveAsset:Array, access=Constructor}
 Exhaust.Pops.Pitch {type=Float, access=Constructor}
 Exhaust.Pops.Shift.Volume {type=Float, access=Constructor}
 Exhaust.Pops.Short {type=WaveAsset:Array, access=Constructor}
 Exhaust.Pops.Volume {type=Float, access=Constructor}
 Exhaust.Throttle.EQ.FilterQ {type=Float, access=Constructor}
 Exhaust.Throttle.EQ.FilterType {type=Enum:BiquadFilterType, access=Constructor}
 Exhaust.Throttle.EQ.OffFreq {type=Float, access=Constructor}
 Exhaust.Throttle.EQ.OffGain {type=Float, access=Constructor}
 Exhaust.Throttle.EQ.On Throttle Frequency {type=Float, access=Constructor}
 Exhaust.Throttle.EQ.OnGain {type=Float, access=Constructor}
 Exhaust.Throttle.InterpTime {type=Time, access=Constructor}
 Exhaust.Throttle.Volume.Off {type=Float, access=Constructor}
 Exhaust.Throttle.Volume.On {type=Float, access=Constructor}
 Fort.Vehicle.Acceleration {type=Float}
 Fort.Vehicle.BlowoffStrength {type=Float}
 Fort.Vehicle.BonusSpeed {type=Float}
 Fort.Vehicle.EnableShifting {type=Bool}
 Fort.Vehicle.Gear {type=Int32}
 Fort.Vehicle.IsGrounded {type=Bool}
 Fort.Vehicle.Load {type=Float}
 Fort.Vehicle.LowFuel {type=Bool}
 Fort.Vehicle.NormalSpeed {type=Float}
 Fort.Vehicle.OnBlowoff {type=Trigger}
 Fort.Vehicle.OnDownshift {type=Trigger}
 Fort.Vehicle.OnThrottleEngaged {type=Trigger}
 Fort.Vehicle.OnThrottleReleased {type=Trigger}
 Fort.Vehicle.OnUpshift {type=Trigger}
 Fort.Vehicle.Pitch {type=Float}
 Fort.Vehicle.RPM {type=Float}
 Fort.Vehicle.Throttle {type=Float}
 GearChangeTime {type=Time, access=Constructor}
 IsLocal {type=Bool}
 Mono {type=Rev Asset, access=Constructor}
 REV.EnableShifting {type=Bool}
 REV.Gear {type=Int32}
 REV.RPM {type=Float}
 REV.Throttle {type=Float}
 REV.Velocity {type=Float}
 REVVolume {type=Float}
 Settings {type=Rev Settings, access=Constructor}
 Speed_Normalized {type=Float}
 UE.Source.OnPlay {type=Trigger}
 UE.Source.Orientation.Azimuth {type=Float}
 UE.Source.Orientation.Elevation {type=Float}
 Upshift {type=Trigger}
]
interfaces [Fort.Character:1.0, Fort.Vehicle:1.0, REV:1.0, UE.OutputFormat.Mono:1.0, UE.Source.Orientation:1.0, UE.Source:1.1]
outputs [
 Downshift {type=Trigger}
 Fundamental Frequency {type=Float}
 Gear {type=Int32}
 RPM {type=Float}
 SpeedNormalized {type=Float}
 Throttle {type=Float}
 "UE.OutputFormat.Mono.Audio:0" {type=Audio}
 "UE.OutputFormat.Stereo.Audio:1" {type=Audio}
 Upshift {type=Trigger}
]
pages [Default {
  input_defaults {
   Blowoff.Volume=-4
   Character.Affiliation=0
   Character.Health=0
   Character.UniqueID=0
   Decibels=10.5
   Downshift=null
   EngineBlowoff=null
   Exhaust.Pops.Blowoff.Volume=0
   Exhaust.Pops.Pitch=0
   Exhaust.Pops.Shift.Volume=0
   Exhaust.Pops.Volume=0
   Exhaust.Throttle.EQ.FilterQ=1
   Exhaust.Throttle.EQ.FilterType=4
   Exhaust.Throttle.EQ.OffFreq=20000
   Exhaust.Throttle.EQ.OffGain=0
   Exhaust.Throttle.EQ.On Throttle Frequency=20000
   Exhaust.Throttle.EQ.OnGain=0
   Exhaust.Throttle.InterpTime=0.1
   Exhaust.Throttle.Volume.Off=0
   Exhaust.Throttle.Volume.On=0
   Fort.Vehicle.Acceleration=0
   Fort.Vehicle.BlowoffStrength=0
   Fort.Vehicle.BonusSpeed=0
   Fort.Vehicle.EnableShifting="<MetaSoundBuilderResult.SUCCEEDED: 0>"
   Fort.Vehicle.Gear=0
   Fort.Vehicle.IsGrounded="<MetaSoundBuilderResult.SUCCEEDED: 0>"
   Fort.Vehicle.Load=0
   Fort.Vehicle.LowFuel=null
   Fort.Vehicle.NormalSpeed=0.3
   Fort.Vehicle.OnBlowoff=null
   Fort.Vehicle.OnDownshift=null
   Fort.Vehicle.OnThrottleEngaged=null
   Fort.Vehicle.OnThrottleReleased=null
   Fort.Vehicle.OnUpshift=null
   Fort.Vehicle.Pitch=0
   Fort.Vehicle.RPM=0
   Fort.Vehicle.Throttle=0
   GearChangeTime=0.1
   IsLocal=null
   Mono=&/EngineAudio_LionsMane
   REV.EnableShifting="<MetaSoundBuilderResult.SUCCEEDED: 0>"
   REV.Gear=0
   REV.RPM=0
   REV.Throttle=0
   REV.Velocity=0
   REVVolume=9.5
   Settings=&/REVSettings_LionsMane
   Speed_Normalized=0
   UE.Source.OnPlay=null
   UE.Source.Orientation.Azimuth=0
   UE.Source.Orientation.Elevation=0
   Upshift=null
   Blowoff.Sound [
    &/Engine_LionsMane_Blowoff_01
    &/Engine_LionsMane_Blowoff_02
    &/Engine_LionsMane_Blowoff_03
    &/Engine_LionsMane_Blowoff_04
    &/Engine_LionsMane_Blowoff_05
   ]
   Exhaust.Pops.Long []
   Exhaust.Pops.Short []
  }
  nodes [
   MSS_EngineAudio_LowSpec_Parent_REV_0 = /VehicleAudio/Patches/MSS_EngineAudio_LowSpec_Parent_REV(
    Blowoff.Sound=^Blowoff.Sound
    Blowoff.Volume=^Blowoff.Volume
    Character.Affiliation=^Character.Affiliation
    Character.Health=^Character.Health
    Character.UniqueID=^Character.UniqueID
    Downshift=^Downshift
    EngineBlowoff=^EngineBlowoff
    Exhaust.Pops.Blowoff.Volume=^Exhaust.Pops.Blowoff.Volume
    Exhaust.Pops.Long=^Exhaust.Pops.Long
    Exhaust.Pops.Pitch=^Exhaust.Pops.Pitch
    Exhaust.Pops.Shift.Volume=^Exhaust.Pops.Shift.Volume
    Exhaust.Pops.Short=^Exhaust.Pops.Short
    Exhaust.Pops.Volume=^Exhaust.Pops.Volume
    Exhaust.Throttle.EQ.FilterQ=^Exhaust.Throttle.EQ.FilterQ
    Exhaust.Throttle.EQ.FilterType=^Exhaust.Throttle.EQ.FilterType
    Exhaust.Throttle.EQ.OffFreq=^Exhaust.Throttle.EQ.OffFreq
    Exhaust.Throttle.EQ.OffGain=^Exhaust.Throttle.EQ.OffGain
    Exhaust.Throttle.EQ.On Throttle Frequency=^Exhaust.Throttle.EQ.On Throttle Frequency
    Exhaust.Throttle.EQ.OnGain=^Exhaust.Throttle.EQ.OnGain
    Exhaust.Throttle.InterpTime=^Exhaust.Throttle.InterpTime
    Exhaust.Throttle.Volume.Off=^Exhaust.Throttle.Volume.Off
    Exhaust.Throttle.Volume.On=^Exhaust.Throttle.Volume.On
    Fort.Vehicle.Acceleration=^Fort.Vehicle.Acceleration
    Fort.Vehicle.BlowoffStrength=^Fort.Vehicle.BlowoffStrength
    Fort.Vehicle.BonusSpeed=^Fort.Vehicle.BonusSpeed
    Fort.Vehicle.EnableShifting=^Fort.Vehicle.EnableShifting
    Fort.Vehicle.Gear=^Fort.Vehicle.Gear
    Fort.Vehicle.IsGrounded=^Fort.Vehicle.IsGrounded
    Fort.Vehicle.Load=^Fort.Vehicle.Load
    Fort.Vehicle.NormalSpeed=^Fort.Vehicle.NormalSpeed
    Fort.Vehicle.OnBlowoff=^Fort.Vehicle.OnBlowoff
    Fort.Vehicle.OnDownshift=^Fort.Vehicle.OnDownshift
    Fort.Vehicle.OnThrottleEngaged=^Fort.Vehicle.OnThrottleEngaged
    Fort.Vehicle.OnThrottleReleased=^Fort.Vehicle.OnThrottleReleased
    Fort.Vehicle.OnUpshift=^Fort.Vehicle.OnUpshift
    Fort.Vehicle.Pitch=^Fort.Vehicle.Pitch
    Fort.Vehicle.RPM=^Fort.Vehicle.RPM
    Fort.Vehicle.Throttle=^Fort.Vehicle.Throttle
    GearChangeTime=^GearChangeTime
    IsLocal=^IsLocal
    Mono=^Mono
    REV.EnableShifting=^REV.EnableShifting
    REV.Gear=^REV.Gear
    REV.RPM=^REV.RPM
    REV.Throttle=^REV.Throttle
    REV.Velocity=^REV.Velocity
    REVVolume=^REVVolume
    Settings=^Settings
    Speed_Normalized=^Speed_Normalized
    UE.Source.OnPlay=^UE.Source.OnPlay
    UE.Source.Orientation.Azimuth=^UE.Source.Orientation.Azimuth
    UE.Source.Orientation.Elevation=^UE.Source.Orientation.Elevation
    Upshift=^Upshift
   )
   DecibelstoLinearGain_0 = UE.Decibels to Linear Gain:Float(Decibels=^Decibels)
   n1 = Fortnite.CrossfadeSegmentManager:Float(In=^Fort.Vehicle.NormalSpeed) {also {defaults {Max=1.1, MaxRange=0, Min=0.2, MinRange=0.6}}}
   WavePlayer_0 = UE.Wave Player:Mono(
    Play=@n1.Activate
    Stop=@n1.Deactivate
    Pitch Shift=MapRange.MapRange:Float(In=^Fort.Vehicle.NormalSpeed) {also {defaults {Out Range A=-4, Out Range B=-2}}}.Out Value
   ) {also {
     defaults {Loop="<MetaSoundBuilderResult.SUCCEEDED: 0>", Wave Asset=&/Engine_LionsMane_Putter_Loop_01}
     version [1,1]
    }}
   AudioMixerAudioMixerMono2_0 = "AudioMixer.Audio Mixer (Mono, 2)"(
    In 1=@MSS_EngineAudio_LowSpec_Parent_REV_0."UE.OutputFormat.Mono.Audio:0"
    Gain 0=UE.Multiply:Float(AdditionalOperands=@DecibelstoLinearGain_0.Linear Gain, PrimaryOperand=@n1.Volume).Out
    In 0=@WavePlayer_0.Out Mono
   )
  ]
  links [
   {dst="@UE.OutputFormat.Mono.Audio:0", src=AudioMixerAudioMixerMono2_0.Out}
   {dst="@Downshift", src=MSS_EngineAudio_LowSpec_Parent_REV_0.Downshift}
   {dst="@Fundamental Frequency", src=MSS_EngineAudio_LowSpec_Parent_REV_0.Fundamental Frequency}
   {dst="@Gear", src=MSS_EngineAudio_LowSpec_Parent_REV_0.Gear}
   {dst="@RPM", src=MSS_EngineAudio_LowSpec_Parent_REV_0.RPM}
   {dst="@SpeedNormalized", src=MSS_EngineAudio_LowSpec_Parent_REV_0.SpeedNormalized}
   {dst="@Throttle", src=MSS_EngineAudio_LowSpec_Parent_REV_0.Throttle}
   {dst="@Upshift", src=MSS_EngineAudio_LowSpec_Parent_REV_0.Upshift}
  ]
 }]
properties {
 assetImportData=&self:AssetImportData
 attenuationSettings=&Vehicle_Engine_Att
 audioPropertiesSheet=&APS_Cosmetic_Vehicle_Engine
 bLooping=true
 duration=10000
 maxDistance=12000
 pitch=1.2599
 priority=80
 qualitySetting=Vehicles_REV_Mono
 sampleRate=48000
 soundClassObject=&Fort_Vehicles_Engine
 soundSubmixObject=&Fortnite_SFX
 volume=1.35
 concurrencySet [&Conc_Veh_EngineAudio_LowSpec, &Conc_Veh_EngineAudio_LowSpec_Actor]
 modulationSettings {
  highpassRouting=Union
  lowpassRouting=Union
  pitchRouting=Union
  volumeRouting=Union
  highpassModulationDestination {bEnableModulation=true, modulators [&PP_FortVehicles_Engine_HPF]}
  lowpassModulationDestination {bEnableModulation=true, modulators [&Patch_FortVehicles_Engine_LPF]}
  pitchModulationDestination {bEnableModulation=true, modulators [&DM_Bus_Vehicle_Engine_Pitch]}
  volumeModulationDestination {
   bEnableModulation=true
   modulators [&DM_VehicleEnginePatch, &Patch_FortVehicles_Engine_Vol]
  }
 }
}
